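# Post-process Anim Blueprint for the LarkEarwig character head that counters head rotation relative to the reference pose.
# When headLimit is set and no emote or custom montage plays, the update negates and scales per-axis head and neck deltas from the reference pose.
# Only head rot feeds the additive bone-space modifiers; both neck modifiers have zero rotation, leaving computed neck values unused.
# Front-end idle-curve and celebration-emote flags are also computed but unused here.
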AnimBlueprint /BRCosmetics/Characters/Player/Male/Large/Heads/M_LRG_LarkEarwig_Head/Meshes/CustomSkeleton/Fortnite_Base_Head_LarkEarwig_AnimB_PostP
target_skeleton=&/Fortnite_Base_Head_LarkEarwig_Skeleton
graphs [
 AnimGraph {nodes [
   n1 = Input Pose()
   n2 = "Transform (Modify) Bone - Bone: neck_02"!(
    ComponentPose="Transform (Modify) Bone - Bone: head"!(
     ComponentPose=Local To Component(LocalPose=@n1)
     bAlphaBoolEnabled:bool=True
     Translation="0,0,0"
     Rotation=Get head rot() -> (head rot:Rotator)
     Scale="1,1,1"
    ) -> (Pose) {alphaInputType=Bool, rotationMode=BMM_Additive, rotationSpace=BCS_BoneSpace, boneToModify {boneName=head}}
    bAlphaBoolEnabled:bool=True
    Translation="0,0,0"
    Rotation="0,0,0"
    Scale="1,1,1"
   ) -> (Pose) {alphaInputType=Bool, rotationMode=BMM_Additive, rotationSpace=BCS_BoneSpace, boneToModify {boneName=neck_02}}
   Output Pose(Result=Component To Local(ComponentPose="Transform (Modify) Bone - Bone: neck_01"!(ComponentPose=@n2, bAlphaBoolEnabled:bool=True, Translation="0,0,0", Rotation="0,0,0", Scale="1,1,1") -> (Pose) {alphaInputType=Bool, rotationMode=BMM_Additive, rotationSpace=BCS_BoneSpace, boneToModify {boneName=neck_01}}))
  ]}
 EventGraph {nodes [
   Event BlueprintUpdateAnimation()
   > n1 = Sequence!(execute) -> (then_0, then_1, then_2:Exec, then_3:Exec)
   > Branch(Condition=Get bIsInFrontEnd())
   > Set Custom Idle Is Playing Float(Custom Idle Is Playing Float:double=GetCurveValue(self=GetOwnerFrontEndAnimInstance(), CurveName=Frontend_Idle_Bypass)) -> (Output_Get:double)
   n2 = GetAnimInstance(self=GetOwnerSkeletalMeshComponent())
   Is Valid(exec=@n1.then_2, InputObject=@n2)
   > Set Celebration Emote(Celebration Emote:bool=Montage_IsPlaying(self=@n2, Montage=/Game/Animation/Game/MainPlayer/Emotes/Emote_Celebration_M)) -> (Output_Get:bool)
   n3 = Get Celebration Emote() -> (Celebration Emote:bool)
   n4 = Get bIsInFrontEnd()
   n5 = Set FEPlayingEmote(execute=@n1.then_3, FEPlayingEmote:bool=AND Boolean(A=NOT Boolean(A=@n3), B=@n4)) -> (Output_Get:bool)
   > Set FEPlayingEmoteOrNoFE(FEPlayingEmoteOrNoFE:bool=OR Boolean(A=@n5, B=NOT Boolean(A=@n4))) -> (Output_Get:bool)
   n6 = Get isCustomMontagePlaying() -> (isCustomMontagePlaying:bool)
   n7 = Get bIsPlayingEmote()
   Branch(
    execute=@n1.then_1
    Condition=AND Boolean(A=NOR Boolean!(A:bool=@n6, B:bool=@n7) -> (ReturnValue:bool), B=Get headLimit() -> (headLimit:bool))
   )
   > n8 = Sequence!(execute) -> (then_0, then_1)
   n9 = GetOwnerSkeletalMeshComponent()
   n10 = BreakRotator(InRot=BreakTransform(InTransform=GetDeltaTransformFromRefPose(self=@n9, BoneName=head, BaseName=neck_02)).Rotation)
   n11 = Get Z percentage() -> (Z percentage:float)
   n12 = Get X percentage() -> (X percentage:float)
   n13 = Get Y percentage() -> (Y percentage:float)
   > Set head rot(head rot:Rotator=MakeRotator(
     Roll=float * float!(A=@n10, B=-1, C:double=@n12) -> (ReturnValue)
     Pitch=float * float(A=@n10.Pitch, B=-1, C=@n13)
     Yaw=float * float(A=@n10.Yaw, B=-1, C=@n11)
    )) -> (Output_Get:Rotator)
   n14 = GetOwnerSkeletalMeshComponent()
   n15 = BreakRotator(InRot=BreakTransform(InTransform=GetDeltaTransformFromRefPose(self=@n14, BoneName=neck_02, BaseName=neck_01)).Rotation)
   n16 = Get X Neck_02 Percentage() -> (X Neck_02 Percentage:double)
   n17 = Get Z Neck_02 Percentage() -> (Z Neck_02 Percentage:double)
   n18 = Get Y Neck_02 Percentage() -> (Y Neck_02 Percentage:double)
   > Set Neck_02 Rot(Neck_02 Rot:Rotator=MakeRotator(Roll=float * float(A=@n15, B=-1, C=@n16), Pitch=float * float(A=@n15.Pitch, B=-1, C=@n18), Yaw=float * float(A=@n15.Yaw, B=-1, C=@n17))) -> (Output_Get:Rotator)
   n19 = GetOwnerSkeletalMeshComponent()
   n20 = BreakRotator(InRot=BreakTransform(InTransform=GetDeltaTransformFromRefPose(self=@n19, BoneName=neck_01, BaseName=spine_05)).Rotation)
   n21 = Get Z Neck Percentage() -> (Z Neck Percentage:double)
   n22 = Get X Neck Percentage() -> (X Neck Percentage:double)
   n23 = Get Y Neck Percentage() -> (Y Neck Percentage:double)
   Set Neck rot(
    execute=@n8.then_1
    Neck rot:Rotator=MakeRotator(Roll=float * float(A=@n20, B=-1, C=@n22), Pitch=float * float(A=@n20.Pitch, B=-1, C=@n23), Yaw=float * float(A=@n20.Yaw, B=-1, C=@n21))
   ) -> (Output_Get:Rotator)
   Event BlueprintBeginPlay() {is=title K2Node_Event}
   > Set headLimit(headLimit:bool=true) -> (Output_Get:bool)
   Reroute Node()
   n24 = GetAnimInstance(self=GetOwnerSkeletalMeshComponent())
   > Is Valid(InputObject=@n24)
   Montage_IsPlaying(self=@n24, Montage=/BRCosmetics/Animation/Game/MainPlayer/Menu/BR/CustomSkins/_s40/TopicProud_M/TopicProud_Lobby_Idle_Montage)
   Montage_IsPlaying(Montage=&TopicProud_Lobby_Idle_Montage)
   Sequence()
   Set isCustomMontagePlaying(isCustomMontagePlaying:bool=false) -> (Output_Get:bool)
  ]}
]
variables [
 Celebration Emote {type=boolean}
 Custom Idle Is Playing Float {type=number}
 FEPlayingEmote {type=boolean}
 FEPlayingEmoteOrNoFE {type=boolean}
 head rot {type=Rotator}
 headLimit {type=boolean}
 isCustomMontagePlaying {type=boolean}
 Neck rot {type=Rotator}
 Neck_02 Rot {type=Rotator}
 X Neck Percentage {type=number, default=1}
 X Neck_02 Percentage {type=number, default=1}
 X percentage {type=number, default=1}
 Y Neck Percentage {type=number, default=0.5}
 Y Neck_02 Percentage {type=number}
 Y percentage {type=number, default=1}
 Z Neck Percentage {type=number, default=0.5}
 Z Neck_02 Percentage {type=number}
 Z percentage {type=number, default=1}
]
